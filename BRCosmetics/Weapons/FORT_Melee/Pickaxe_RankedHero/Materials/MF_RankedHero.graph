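# Material function for the RankedHero pickaxe that layers a reflective thin-film look onto incoming material attributes.
# It derives an area mask from a TechArt mask input or the sign of incoming emissive, selected by parameter.
# It builds bump-offset hexagon texture, refracted skylight environment and fake light-vector reflection terms, plus blue-tinted thin-film emissive.
# Fresnel and mask blends override base color, metallic, specular and roughness, add emissive, then feed a shared thin-film function.
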MaterialFunction /BRCosmetics/Weapons/FORT_Melee/Pickaxe_RankedHero/Materials/MF_RankedHero
graph {nodes [
  Constant()
  n1 = TextureSampleParameter2D(UVs=BumpOffset(Height=ScalarParameter() {defaultValue=1, parameterName=RH_BumpOffsetHeight})) {
   parameterName=Diffuse
   samplerSource=SSM_Wrap_WorldGroupSettings
   texture=/BRCosmetics/Weapons/FORT_Melee/Pickaxe_RankedHero/Textures/T_RankedHero_Pickaxe_D
  }
  n2 = MaterialFunctionCall(Power=ScalarParameter() {defaultValue=0.25, parameterName=RH_BumpOffsetFresnelPower}) {materialFunction=&Fresnel_Function}
  n3 = LinearInterpolate(A=@n1.RGB, B=NamedRerouteUsage(), Alpha=@n2.Result)
  n4 = NamedRerouteUsage()
  n5 = Fresnel(ExponentIn=ScalarParameter() {defaultValue=0.3, parameterName=RH_EnvMapFresnelPower})
  n6 = ScalarParameter() {defaultValue=0.5, parameterName=RH_Roughness_Low}
  n7 = ScalarParameter() {defaultValue=15, parameterName=RH_Roughness_High}
  n8 = NamedRerouteUsage()
  n9 = Fresnel(Normal=PixelNormalWS()) {exponent=1}
  n10 = Multiply(A=@n9, B=LinearInterpolate(Alpha=NamedRerouteUsage()) {constA=0.5, constB=0}) {constB=0.5}
  n11 = Multiply(
   A=LinearInterpolate(
    B=TextureSample(UVs=@n10) {texture=/Game/Global/Textures/Default/Thinfilm/T_M_MED_SykoBlue_Body_ThinFilm}.RGB
    Alpha=Add(A=@n9) {constB=-0.2}
   )
   B=Constant3Vector() {constant=<rgba 0 0.5495 1.5938 1>}
  )
  n12 = NamedRerouteUsage()
  n13 = SetMaterialAttributes!(
   MaterialAttributes=NamedRerouteUsage()
   BaseColor=LinearInterpolate(A=NamedRerouteUsage(), B=LinearInterpolate(A=@n3, B=@n4, Alpha=@n5) {constAlpha=1}, Alpha=NamedRerouteUsage())
   Metallic=LinearInterpolate(A=NamedRerouteUsage(), B=ScalarParameter() {parameterName=RH_Metallic}, Alpha=NamedRerouteUsage())
   Specular=LinearInterpolate(A=NamedRerouteUsage(), B=ScalarParameter() {defaultValue=1, parameterName=RH_Specular}, Alpha=NamedRerouteUsage())
   Roughness=LinearInterpolate(A=NamedRerouteUsage(), B=LinearInterpolate(A=@n6, B=@n7, Alpha=@n8), Alpha=NamedRerouteUsage())
   EmissiveColor=Add(A=NamedRerouteUsage(), B=Multiply(A=@n11, B=@n12))
  ) {attributeSetTypes [
    #1
    #2
    #3
    #4
    #5
   ]}
  FunctionOutput(None=MaterialFunctionCall!(Material Attributes=@n13, Fresnel, Alpha=NamedRerouteUsage()) {materialFunction=/Game/Characters/Player/Female/Medium/Bodies/F_MED_SkeleProbe/Materials/MF_ThinFilmDuplicate_TechArt}.Result)
  MaterialFunctionCall(Power=Constant() {r=1}) {materialFunction=&Fresnel_Function}
  n14 = TextureSample() {samplerType=SAMPLERTYPE_Masks, texture=/BRCosmetics/Gadgets/Assets/VinderTech_GliderChute/Glider_Season_36/Textures/T_Season_36_Glider_FX2}
  n15 = ComponentMask(None=FunctionInput(Preview=@n14.RGB) {inputName=TechArtMask}) {r=true}
  n16 = GetMaterialAttributes(None=FunctionInput() {inputName=MatIn, inputType=FunctionInput_MaterialAttributes}) {attributeGetTypes [
    #5
    #6
    #1
    #2
    #3
    #4
   ]}
  n17 = Sign(None=@n16.EmissiveColor)
  NamedRerouteDeclaration(None=LinearInterpolate(A=@n15, B=Saturate(None=@n17), Alpha=ScalarParameter() {defaultValue=1, parameterName=RH_UseEmissiveTextureAsMask}) {constAlpha=0}) {name=CyanAreaMask, nodeColor=<rgba 0.3125 1 0 1>}
  NamedRerouteDeclaration(None=@n16.MaterialAttributes) {name=MatIn, nodeColor=<rgba 1 0 0.6328 1>}
  NamedRerouteDeclaration(None=@n16.Specular) {name=SpecularIn, nodeColor=<rgba 1 0 0.8437 1>}
  NamedRerouteDeclaration(None=@n16.Roughness) {name=RoughnessIn, nodeColor=<rgba 0 1 0.8359 1>}
  NamedRerouteDeclaration(None=@n15) {name=TechArtMask, nodeColor=<rgba 0.9453 0 1 1>}
  NamedRerouteDeclaration(None=@n16.Normal) {name=NormalIn, nodeColor=<rgba 0.875 1 0 1>}
  n18 = DotProduct(A=AtmosphericLightVector(), B=NamedRerouteUsage())
  n19 = ScalarParameter() {defaultValue=5, parameterName=RH_SDNPower}
  NamedRerouteDeclaration(None=Multiply(A=Power(Base=@n18, Exp=@n19), B=ScalarParameter() {defaultValue=0.1, parameterName=RH_SDNStrength})) {name=FakeReflection, nodeColor=<rgba 0.8047 0 1 1>}
  NamedRerouteDeclaration(None=@n16.EmissiveColor) {name=EmissiveIn, nodeColor=<rgba 0 1 0.4844 1>}
  NamedRerouteDeclaration(None=@n16.BaseColor) {name=BaseColorIn, nodeColor=<rgba 0.4062 0 1 1>}
  NamedRerouteDeclaration(None=@n2.Result) {name=Fresnel, nodeColor=<rgba 0.4766 1 0 1>}
  n20 = CameraVectorWS()
  n21 = Transform(None=Constant3Vector() {constant=<rgba 0 0 1 0>}) {transformSourceType=TRANSFORMSOURCE_Local}
  n22 = MaterialFunctionCall!(
   Ray Direction=Multiply(A=StaticSwitchParameter(True=@n20, False=@n21) {defaultValue=true, parameterName=UseCameraVectorAsRayDirection}) {constA=-1, constB=-1}
   Surface Normal
   Refractive Index Origin
   Refractive Index Target=ScalarParameter() {defaultValue=0.74, parameterName=RH_RefractionTarget}
  ) {materialFunction=/Engine/Functions/Engine_MaterialFunctions01/Vectors/Refract}
  NamedRerouteDeclaration(None=Multiply(A=SkyLightEnvMapSample(Direction=@n22.Vector), B=ScalarParameter() {defaultValue=0.25, parameterName=RH_EnvMapMult}) {constB=0.25}) {name=EnvironmentMap, nodeColor=<rgba 1 0 0.7969 1>}
  NamedRerouteDeclaration(None=Power(Base=@n1.R) {constExponent=1.5}) {name=HexagonLines, nodeColor=<rgba 0.5469 1 0 1>}
  NamedRerouteDeclaration(None=@n16.Metallic) {name=MetallicIn, nodeColor=<rgba 0.5234 1 0 1>}
  PixelNormalWS()
 ]}
